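# Niagara smoke burst effect for the OvenDrastic emote, pairing flat ground rings with smoke puffs thrown out to either side.
# A ring emitter bursts short-lived, upward-facing sprites that expand and fade.
# After a short delay, mirrored left and right emitters burst spinning SubUV puffs along opposite Y directions, shaped by gravity and a user drag curve.
# All emitters scale color and alpha by a system time-of-day color, and dynamic material parameters animate the sprites.
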NiagaraSystem /BRCosmetics/Animation/Game/MainPlayer/Emotes/OvenDrastic/FX/NS_Emote_OvenDrastic_SmokegroundBurst1
dependencies {
 usedDataInterfaces [&NiagaraDataInterfaceSpriteRendererInfo, &NiagaraDataInterfaceVector2DCurve]
 usedDynamicInputs [
  &SimulationPosition
  &MakeFloatFromVector2d
  &Multiply_Vector2
  &Lerp_Float
  &MakeCustomFloatFromBool
  &SetBoolByFloatComparison
  &Abs_Float
  &MakeFloatFromVector
  &ClampFloat
  &FNTimeOfDayColor
 ]
 usedModules [
  &SystemState
  &self:SystemSpawnScript.SystemScriptSource.SystemScriptGraph.NiagaraNodeAssignment_1.SetVariables_#1
 ]
}
emitters {
 NE_SmokeGroundRings {
  emitterUpdateScript [
   &EmitterState(Inactive Response=::NewEnumerator1, Life Cycle Mode=::NewEnumerator1, Loop Behavior=::NewEnumerator1, Loop Duration=0.45, UseLoopDelay=false)
   &SpawnBurst_Instantaneous(Spawn Count=3, Use Spawn Probability=false, UseLoopCountLimit=false)
  ]
  particleSpawnScript [
   &InitializeParticle(
    Color=<rgba 1 0.7225 0.5208 1.5>
    Lifetime=0.35
    Mass Mode=::NewEnumerator1
    Position Offset=<60 -50 0>
    Sprite Rotation Angle=&UniformRangedFloat(Minimum=-180, Maximum=180)
    Sprite Rotation Mode=::NewEnumerator0
    Sprite Size=&Vector2DFromFloat(Value=400)
    Sprite Size Mode=::NewEnumerator4
    UsePositionOffset=true
    Write Material Random=true
   ) {module=InitializeParticle001}
   /SetVariables(Particles.SpriteFacing=<0 0 1>) {isSetParameters=true}
  ]
  particleUpdateScript [
   &ParticleState()
   &DynamicMaterialParameters(
    Index 0 Param 1=&FloatFromCurve(FloatCurve=DataInterface() {propertyValues {Curve {keys [<at 0 0.5>,<at 1 2>]}}})
    Module.Param0WriteEnabled=true
    Module.Param1WriteEnabled=false
    Module.Param2WriteEnabled=false
    Module.Param3WriteEnabled=false
   )
   &ScaleSpriteSize(Scale Factor=&Vector2DFromFloat(Value=&FloatFromCurve(FloatCurve=DataInterface() {propertyValues {Curve {keys [
          {arriveTangent=35.6763, interpMode=RCIM_Cubic, leaveTangent=3.7835, tangentMode=RCTM_User, time=0, value=0}
          {interpMode=RCIM_Cubic, leaveTangent=0.6903, time=1.0011, value=1}
         ]}}})))
   &CameraOffset(Camera Offset Amount=20)
   &ScaleColor(
    Scale Alpha=&FloatFromCurve(FloatCurve=DataInterface() {propertyValues {Curve {keys [
         {arriveTangent=0.0112, interpMode=RCIM_Cubic, leaveTangent=0.0112, tangentMode=RCTM_User, time=0, value=0}
         {arriveTangent=0.002, interpMode=RCIM_Cubic, leaveTangent=0.002, time=0.4, value=0.4}
         {interpMode=RCIM_Cubic, leaveTangent=-4.9981, time=1.0011, value=0.002}
        ]}}})
    Scale RGB=&MakeVectorFromLinearColorRGB(Color=$System.TOD Color)
    ScaleA=true
    ScaleRGB=true
   )
   &InheritVelocity(
    Inherited Velocity Amount Scale=<-1 -1 -1>
    Inherited Velocity Speed Limit=5
    Source Speed Threshold=3
    Speed Limit=true
    UseSpeedThreshold=true
   )
   &SolveForcesAndVelocity(Clamp Velocity=false, Limit Acceleration=false)
   &ScaleColor(
    Scale Alpha=&MakeFloatFromLinearColor(LinearColor=$System.TOD Color, Channel:ENiagaraFloatFromLinearColorOptions=::NewEnumerator3)
    ScaleA=true
    ScaleRGB=false
   ) {module=ScaleColor002}
  ]
  rendererClasses [&NiagaraSpriteRendererProperties]
  renderers [{rendererClass=&NiagaraSpriteRendererProperties}]
 }
 NE_SmokeLeft {
  emitterUpdateScript [
   &EmitterState(Life Cycle Mode=::NewEnumerator1, Loop Behavior=::NewEnumerator1, Loop Delay=0.18, Loop Duration=2, UseLoopDelay=true)
   &SpawnBurst_Instantaneous(Spawn Count=10, Use Spawn Probability=false, UseLoopCountLimit=false)
  ]
  particleSpawnScript [
   &InitializeParticle(
    Color=$User.Linear Color
    Lifetime=&UniformRangedFloat(Minimum=0.4, Maximum=0.5)
    Position Offset=<60 0 35>
    Sprite Size=&UniformRangedVector2D(Minimum=<25 25>, Maximum=<35 35>)
    Sprite Size Mode=::NewEnumerator4
    UsePositionOffset=true
    Write Material Random=true
   )
   &ApplyInitialForces()
   &SubUVAnimation(
    Evaluation Type=::NewEnumerator0
    "Sprite Renderer "=DataInterface() {propertyValues {SpriteRenderer=&self:NE_SmokeLEFt_0.NiagaraSpriteRendererProperties_0}}
    SubUV Animation Mode:ENiagara_SubUVLookupModeV2=::NewEnumerator2
    UseEndFrame=false
    UseStartFrame=false
   )
   &AddVelocity(Velocity=&RandomRangeVector(Minimum=<0 50 0>, Maximum=<0 400 0>)) {module=AddVelocity001}
   &ShapeLocation(MaskSpawnGroup=false, Non Uniform Scale=<1.5 2 0>, Override Seed=false, Sphere Radius=20, UseOwnerScale=true) {module=ShapeLocation001}
  ]
  particleUpdateScript [
   &ParticleState()
   &SpriteRotationRate(Rotation Rate=&RandomRangeFloat(Minimum=300, Maximum=400))
   &ScaleSpriteSize(Scale Factor=&Vector2DFromCurve(
     Vector2Curve=DataInterface() {propertyValues {
       XCurve {keys [
         {arriveTangent=2.5428, interpMode=RCIM_Cubic, leaveTangent=2.5428, tangentMode=RCTM_User, time=0, value=1}
         {arriveTangent=0.1281, interpMode=RCIM_Cubic, leaveTangent=0.1281, tangentMode=RCTM_User, time=1, value=2}
        ]}
       YCurve {keys [
         {arriveTangent=2.8823, interpMode=RCIM_Cubic, leaveTangent=2.8823, tangentMode=RCTM_User, time=0, value=1}
         {arriveTangent=2.0297, interpMode=RCIM_Cubic, leaveTangent=2.0297, tangentMode=RCTM_User, time=1, value=2}
        ]}
      }}
     ScaleCurve=&Multiply_Vector2D_ByFloat(
      Float=&RandomRangeFloat(Minimum=1, Maximum=2)
      Vector2D=&RandomRangeVector2D(Minimum=<0.6 0.75>, Maximum=<0.8 1>)
     )
    ))
   &DynamicMaterialParameters(
    Index 0 Param 1=20
    Index 0 Param 2=1
    Index 0 Param 3=1
    Index 0 Param 4=&FloatFromCurve(FloatCurve=DataInterface() {propertyValues {Curve {keys [{tangentMode=RCTM_User, time=0.1, value=0}, {tangentMode=RCTM_User, time=1, value=0.5}]}}})
    Module.Param0WriteEnabled=true
    Module.Param1WriteEnabled=true
    Module.Param2WriteEnabled=true
    Module.Param3WriteEnabled=true
   ) {module=DynamicMaterialParameters001}
   &Drag(Drag=&FloatFromCurve(FloatCurve=$User.drag), Use Linear Drag=true, Use Rotational Drag=true)
   &GravityForce(Gravity=&RandomRangeVector(Minimum=<0 0 -940>, Maximum=<1 1 -980>))
   &SolveForcesAndVelocity(Clamp Velocity=false, Limit Acceleration=false)
   &ScaleColor(
    Scale Alpha=&FloatFromCurve(
     FloatCurve=DataInterface() {propertyValues {Curve {keys [
         {arriveTangent=9.0725, interpMode=RCIM_Cubic, leaveTangent=9.0725, tangentMode=RCTM_User, time=0, value=0}
         {arriveTangent=-2.5249, interpMode=RCIM_Cubic, leaveTangent=-2.5249, tangentMode=RCTM_User, time=0.3, value=0.9847}
         {arriveTangent=-0.0405, interpMode=RCIM_Cubic, leaveTangent=-0.0405, tangentMode=RCTM_User, time=1, value=0}
        ]}}}
     Scale Curve=0.8
    )
    Scale RGB=&MakeVectorFromLinearColorRGB(Color=$System.Tint)
    ScaleA=true
    ScaleRGB=false
   )
   &ScaleColor(
    Scale Alpha=&MakeFloatFromLinearColor(LinearColor=$System.TOD Color, Channel:ENiagaraFloatFromLinearColorOptions=::NewEnumerator3)
    Scale RGB=&MakeVectorFromLinearColorRGB(Color=$System.TOD Color)
    ScaleA=true
    ScaleRGB=true
   ) {module=ScaleColor001}
   &DynamicMaterialParameters(
    Index 0 Param 1=&Add_Float(
     A=&RandomRangeFloat(Minimum=0, Maximum=63)
     B=&FloatFromCurve(
      FloatCurve=DataInterface() {propertyValues {Curve {keys [
          <at 0 0 RCIM_Cubic>
          {arriveTangent=1.5947, interpMode=RCIM_Cubic, leaveTangent=1.5947, tangentMode=RCTM_User, time=0.2557, value=0.5257}
          <at 1 1 RCIM_Cubic>
         ]}}}
      CurveIndex=$Particles.Age
      Scale Curve=0.25
     )
    )
    Module.Index0Float0Write=true
    Module.Index0Float1Write=true
    Module.Index0Float2Write=true
    Module.Index0Float3Write=true
   )
  ]
  rendererClasses [&NiagaraSpriteRendererProperties]
  renderers [
   {rendererClass=&NiagaraSpriteRendererProperties}
   {rendererClass=&NiagaraSpriteRendererProperties, rendererIndex=1}
  ]
 }
 NE_SmokeRight {
  emitterUpdateScript [
   &EmitterState(Life Cycle Mode=::NewEnumerator1, Loop Behavior=::NewEnumerator1, Loop Delay=0.18, Loop Duration=2, UseLoopDelay=true)
   &SpawnBurst_Instantaneous(Spawn Count=10, Use Spawn Probability=false, UseLoopCountLimit=false)
  ]
  particleSpawnScript [
   &InitializeParticle(
    Color=$User.Linear Color
    Lifetime=&UniformRangedFloat(Minimum=0.4, Maximum=0.5)
    Position Offset=<60 -100 35>
    Sprite Size=&UniformRangedVector2D(Minimum=<20 20>, Maximum=<30 30>)
    Sprite Size Mode=::NewEnumerator4
    UsePositionOffset=true
    Write Material Random=true
   )
   &ApplyInitialForces()
   &SubUVAnimation(
    Evaluation Type=::NewEnumerator0
    "Sprite Renderer "=DataInterface() {propertyValues {SpriteRenderer=&self:NE_SmokeRight_0.NiagaraSpriteRendererProperties_0}}
    SubUV Animation Mode:ENiagara_SubUVLookupModeV2=::NewEnumerator2
    UseEndFrame=false
    UseStartFrame=false
   )
   &AddVelocity(Velocity=&RandomRangeVector(Minimum=<0 -50 0>, Maximum=<0 -400 0>)) {module=AddVelocity001}
   &ShapeLocation(MaskSpawnGroup=false, Non Uniform Scale=<1 1.5 0>, Override Seed=false, Sphere Radius=20, UseOwnerScale=true) {module=ShapeLocation001}
  ]
  particleUpdateScript [
   &ParticleState()
   &SpriteRotationRate(Rotation Rate=&RandomRangeFloat(Minimum=-300, Maximum=-400))
   &ScaleSpriteSize(Scale Factor=&Vector2DFromCurve(
     Vector2Curve=DataInterface() {propertyValues {
       XCurve {keys [
         {arriveTangent=2.5428, interpMode=RCIM_Cubic, leaveTangent=2.5428, tangentMode=RCTM_User, time=0, value=1}
         {arriveTangent=0.1281, interpMode=RCIM_Cubic, leaveTangent=0.1281, tangentMode=RCTM_User, time=1, value=2}
        ]}
       YCurve {keys [
         {arriveTangent=2.8823, interpMode=RCIM_Cubic, leaveTangent=2.8823, tangentMode=RCTM_User, time=0, value=1}
         {arriveTangent=2.0297, interpMode=RCIM_Cubic, leaveTangent=2.0297, tangentMode=RCTM_User, time=1, value=2}
        ]}
      }}
     ScaleCurve=&Multiply_Vector2D_ByFloat(
      Float=&RandomRangeFloat(Minimum=1, Maximum=2)
      Vector2D=&RandomRangeVector2D(Minimum=<0.6 0.75>, Maximum=<0.8 1>)
     )
    ))
   &DynamicMaterialParameters(
    Index 0 Param 1=20
    Index 0 Param 2=1
    Index 0 Param 3=1
    Index 0 Param 4=&FloatFromCurve(FloatCurve=DataInterface() {propertyValues {Curve {keys [{tangentMode=RCTM_User, time=0.1, value=0}, {tangentMode=RCTM_User, time=1, value=0.5}]}}})
    Module.Param0WriteEnabled=true
    Module.Param1WriteEnabled=true
    Module.Param2WriteEnabled=true
    Module.Param3WriteEnabled=true
   ) {module=DynamicMaterialParameters001}
   &Drag(Drag=&FloatFromCurve(FloatCurve=$User.drag), Use Linear Drag=true, Use Rotational Drag=true)
   &GravityForce(Gravity=<0 0 -940>)
   &SolveForcesAndVelocity(Clamp Velocity=false, Limit Acceleration=false)
   &ScaleColor(
    Scale Alpha=&FloatFromCurve(
     FloatCurve=DataInterface() {propertyValues {Curve {keys [
         {arriveTangent=9.0725, interpMode=RCIM_Cubic, leaveTangent=9.0725, tangentMode=RCTM_User, time=0, value=0}
         {arriveTangent=-2.5249, interpMode=RCIM_Cubic, leaveTangent=-2.5249, tangentMode=RCTM_User, time=0.3, value=0.9847}
         {arriveTangent=-0.0405, interpMode=RCIM_Cubic, leaveTangent=-0.0405, tangentMode=RCTM_User, time=1, value=0}
        ]}}}
     Scale Curve=0.8
    )
    Scale RGB=&MakeVectorFromLinearColorRGB(Color=$System.Tint)
    ScaleA=true
    ScaleRGB=false
   )
   &ScaleColor(
    Scale Alpha=&MakeFloatFromLinearColor(LinearColor=$System.TOD Color, Channel:ENiagaraFloatFromLinearColorOptions=::NewEnumerator3)
    Scale RGB=&MakeVectorFromLinearColorRGB(Color=$System.TOD Color)
    ScaleA=true
    ScaleRGB=true
   ) {module=ScaleColor001}
   &DynamicMaterialParameters(
    Index 0 Param 1=&Add_Float(
     A=&RandomRangeFloat(Minimum=0, Maximum=63)
     B=&FloatFromCurve(
      FloatCurve=DataInterface() {propertyValues {Curve {keys [
          <at 0 0 RCIM_Cubic>
          {arriveTangent=1.5947, interpMode=RCIM_Cubic, leaveTangent=1.5947, tangentMode=RCTM_User, time=0.2557, value=0.5257}
          <at 1 1 RCIM_Cubic>
         ]}}}
      CurveIndex=$Particles.Age
      Scale Curve=0.25
     )
    )
    Module.Index0Float0Write=true
    Module.Index0Float1Write=true
    Module.Index0Float2Write=true
    Module.Index0Float3Write=true
   )
  ]
  rendererClasses [&NiagaraSpriteRendererProperties]
  renderers [
   {rendererClass=&NiagaraSpriteRendererProperties}
   {rendererClass=&NiagaraSpriteRendererProperties, rendererIndex=1}
  ]
 }
}
summary {
 systemName=NS_Emote_OvenDrastic_SmokegroundBurst1
 emitters [
  {emitterName=NE_SmokeGroundRings, rendererClasses [&NiagaraSpriteRendererProperties]}
  {emitterName=NE_SmokeLeft, rendererClasses [&NiagaraSpriteRendererProperties]}
  {emitterName=NE_SmokeRight, rendererClasses [&NiagaraSpriteRendererProperties]}
 ]
 userVariables [
  User.bIsFrontEnd {type=NiagaraBool}
  User.bIsFrontEndPreview {type=NiagaraBool}
  User.bIsInWater {type=NiagaraBool}
  User.drag {
   type=NiagaraDataInterfaceCurve
   description=""
   defaultValue {
    struct=NiagaraExt_VariableValue_DataInterface
    value {
     dataInterface=&self:NiagaraDataInterfaceCurve_0
     dataInterfaceClass=&NiagaraDataInterfaceCurve
    }
   }
  }
  User.FadeMaxDistance {type=NiagaraFloat, defaultValue=0, description=""}
  User.Linear Color {type=LinearColor, description="", defaultValue {struct=LinearColor, value=<rgba 0.3073 0.1657 0.1698 0.15>}}
 ]
}
