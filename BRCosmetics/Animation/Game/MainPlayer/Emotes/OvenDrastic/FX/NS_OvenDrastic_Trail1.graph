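# Niagara ribbon trail effect for the OvenDrastic main player emote, drawing a bright blue, short-lived trail from skeletal animation.
# A single Trail emitter uses the custom AnimTrail modules to spawn, initialize from bone or socket indices, and update the ribbon.
# Over each particle's life the ribbon narrows and fades, while curves, random ranges and a waveform drive four dynamic material parameter sets.
# A color scale linked to the lightning control user color is disabled.
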NiagaraSystem /BRCosmetics/Animation/Game/MainPlayer/Emotes/OvenDrastic/FX/NS_OvenDrastic_Trail1
dependencies {
 usedDataInterfaces [&NiagaraDataInterfaceCurve]
 usedDynamicInputs [&FNTimeOfDayScalar]
 usedModules [
  &SystemState
  &self:SystemSpawnScript.SystemScriptSource.SystemScriptGraph.NiagaraNodeAssignment_0.SetVariables_#1
 ]
}
emitters {Trail {
  emitterUpdateScript [
   &EmitterState(Life Cycle Mode=::NewEnumerator1, UseLoopDelay=true) {module=EmitterState001}
   &FN_SpawnAnimTrail_v2()
  ]
  particleSpawnScript [
   /Game/Effects/Niagara/Modules/AnimTrail/FN_InitializeAnimTrail/FN_InitializeAnimTrailBoneIndecies(Second Bone or Socket Index=1)
   /Game/Effects/Niagara/Modules/AnimTrail/FN_InitializeAnimTrail/FN_InitializeAnimTrail(AnimTrailWidth:NiagaraFloat=&Multiply_Float(A=1.5, B=1.35), Color=<rgba 0 0.318 2 2>, Lifetime=0.7)
  ]
  particleUpdateScript [
   &FN_UpdateAnimTrail_v2()
   &ScaleRibbonWidth(Ribbon Width Scale=&FloatFromCurve(
     FloatCurve=DataInterface() {propertyValues {Curve {keys [
         {interpMode=RCIM_Cubic, tangentMode=RCTM_User, time=0, value=1.1}
         {arriveTangent=-2.3701, interpMode=RCIM_Cubic, leaveTangent=-2.3701, tangentMode=RCTM_User, time=1, value=0}
        ]}}}
     Scale Curve=11
    ))
   &SpriteFacingAndAlignment(Sprite Facing=$Particles.RibbonFacing, WriteAlignment=false, WriteFacing=true)
   &ScaleColor(
    Scale Alpha=&FloatFromCurve(FloatCurve=DataInterface() {propertyValues {Curve {keys [<at 0 1>,<at 1 0>]}}})
    ScaleA=true
    ScaleRGB=true
   ) {module=ScaleColor002}
   /SetVariables(Particles.float:NiagaraFloat=&Multiply_Float(A=$Particles.RibbonWidth, B=1.4)) {isSetParameters=true}
   &ParticleState()
   &DynamicMaterialParameters(
    Index 0 Param 2=&RandomRangeFloat(Minimum=0, Maximum=0)
    Index 1 Param 1=&FloatFromCurve(FloatCurve=DataInterface() {propertyValues {Curve {keys [<at 0 0.5>,<at 1 1>]}}})
    Index 1 Param 3=&FloatFromCurve(FloatCurve=DataInterface() {propertyValues {Curve {keys [{interpMode=RCIM_Cubic, tangentMode=RCTM_Break, time=0.5, value=-0.5}, <at 1 0.1>]}}})
    Index 2 Param 1=&Multiply_Float(
     A=&FloatFromCurve(
      FloatCurve=DataInterface() {propertyValues {Curve {keys [{interpMode=RCIM_Cubic, leaveTangent=8.1085, tangentMode=RCTM_Break, time=0.8568, value=1}, {tangentMode=RCTM_User, time=1, value=0}]}}}
      Scale Curve=0.5
     )
     B=&Divide_Float(
      A=&MakeFloatFromVector(Vector=$Particles.Scale, Channel=::NewEnumerator0)
      B=&RandomRangeFloat(Minimum=10, Maximum=15)
     )
    )
    Index 2 Param 2=&FloatFromCurve(
     FloatCurve=DataInterface() {propertyValues {Curve {keys [
         {interpMode=RCIM_Cubic, tangentMode=RCTM_Break, time=0.7, value=0}
         {arriveTangent=-0.0859, interpMode=RCIM_Cubic, tangentMode=RCTM_Break, time=1, value=-1}
        ]}}}
     Scale Curve=50
    )
    Index 2 Param 3=&Waveform(
     Global Input Position=&RandomRangeFloat(Minimum=0, Maximum=10)
     Global Phase=$Particles.Age
     "Waveform [1]"=::NewEnumerator12
     "[1] Phase Scale"=2
     "[1] Frequency"=1
     "[1] Amplitude Min/Max"=<0 1.5>
    )
    Index 2 Param 4=25
    Index 3 Param 1=&FloatFromCurve(
     FloatCurve=DataInterface() {propertyValues {Curve {keys [
         {interpMode=RCIM_Cubic, leaveTangent=-0.0025, tangentMode=RCTM_Break, time=0, value=0}
         {arriveTangent=1.8528, interpMode=RCIM_Cubic, tangentMode=RCTM_Break, time=1, value=1}
        ]}}}
     Scale Curve=&RandomRangeFloat(Minimum=0, Maximum=0)
    )
    Index 3 Param 2=&FloatFromCurve(
     FloatCurve=DataInterface() {propertyValues {Curve {keys [
         {interpMode=RCIM_Cubic, leaveTangent=-0.0025, tangentMode=RCTM_Break, time=0, value=0.5}
         {arriveTangent=-0.0316, interpMode=RCIM_Cubic, tangentMode=RCTM_Break, time=1, value=-0.2}
        ]}}}
     Scale Curve=0
    )
    Index 3 Param 3=&FloatFromCurve(
     FloatCurve=DataInterface() {propertyValues {Curve {keys [
         {interpMode=RCIM_Cubic, leaveTangent=1.1034, tangentMode=RCTM_Break, time=0, value=-1}
         {arriveTangent=2.2399, tangentMode=RCTM_User, time=0.5, value=0.125}
        ]}}}
     Scale Curve=0
    )
    Index 3 Param 4=&FloatFromCurve(FloatCurve=DataInterface() {propertyValues {Curve {keys [
         {interpMode=RCIM_Cubic, leaveTangent=1.8089, tangentMode=RCTM_Break, time=0.3, value=-0.5}
         {arriveTangent=2.0757, interpMode=RCIM_Cubic, tangentMode=RCTM_Break, time=0.9, value=1}
        ]}}})
    Module.Index0Param0Write=true
    Module.Index0Param1Write=true
    Module.Index0Param2Write=true
    Module.Index0Param3Write=true
    Module.Index1Param0Write=true
    Module.Index1Param1Write=true
    Module.Index1Param2Write=true
    Module.Index1Param3Write=true
    Module.Index2Param0Write=true
    Module.Index2Param1Write=true
    Module.Index2Param2Write=true
    Module.Index2Param3Write=true
    Module.Index3Param0Write=true
    Module.Index3Param1Write=true
    Module.Index3Param2Write=true
    Module.Index3Param3Write=true
    Write Parameter Index 1=true
    Write Parameter Index 2=true
    Write Parameter Index 3=true
   ) {module=DynamicMaterialParameters001}
   &ScaleColor(
    Color Value To Scale=Linked() {linkedVariable {name="User.lightning control ", type=LinearColor}}
    Scale Alpha=0
    Scale RGB=&MakeVector(
     X=0.5
     Y=&Multiply_Float(
      A=&MakeFloatFromLinearColor(
       LinearColor=Linked() {linkedVariable {name="User.lightning control ", type=LinearColor}}
       Channel:ENiagaraFloatFromLinearColorOptions=::NewEnumerator0
      )
      B=0.4
     )
     Z=0
    )
    ScaleA=false
    ScaleRGB=true
   ) {enabled=false}
  ]
  rendererClasses [&NiagaraRibbonRendererProperties]
  renderers [{rendererClass=&NiagaraRibbonRendererProperties}]
 }}
summary {
 systemName=NS_OvenDrastic_Trail1
 emitters [{emitterName=Trail, rendererClasses [&NiagaraRibbonRendererProperties]}]
 userVariables [
  User.bIsFrontEndPreview {type=NiagaraBool}
  User.BPOpacityMult {type=NiagaraFloat, defaultValue=0, description=""}
  User.CurrentExhaust {type=Vector3f, description="", defaultValue {struct=Vector3f, value=<0 0 0>}}
  User.DebugVelocity {type=NiagaraFloat, defaultValue=0, description=""}
  User.IsR2 {type=NiagaraBool, defaultValue=true}
  "User.lightning control " {type=LinearColor, description="", defaultValue {struct=LinearColor, value=<rgba 0.5 0.05 0 0>}}
  User.PlayerSpeed {type=NiagaraFloat, defaultValue=0, description=""}
  User.RibbonIDOffset {type=NiagaraInt32}
  User.SkeletalMesh {
   type=NiagaraDataInterfaceSkeletalMesh
   defaultValue {
    struct=NiagaraExt_VariableValue_DataInterface
    value {dataInterface=&self:NiagaraDataInterfaceSkeletalMesh_0}
   }
  }
  User.TrailLengthScale {type=NiagaraFloat, defaultValue=0, description=""}
  User.TrailWidth {type=NiagaraFloat, defaultValue=0.4, description=""}
  User.Variant_Color {type=LinearColor, description="", defaultValue {struct=LinearColor, value=<rgba 0.1562 0.1698 1 2>}}
 ]
}
