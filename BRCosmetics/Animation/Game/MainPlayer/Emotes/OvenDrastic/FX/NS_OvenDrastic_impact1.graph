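# Niagara impact effect for the OvenDrastic emote: a short, mostly blue burst of glow, flash, arcs and sparks.
# ArcStraights fires white mesh streaks fanned around the owner's X axis, animated through curve-driven dynamic material parameters.
# AirGlow, Flare (with light), Flash001 (sprite plus mesh) and distance-culled GPU sparks follow after User.exp_Delay.
# User.Color tints only AirGlow, via UseCustomColor; Color_Blade_B colors the light; UseCustomColor1 is unreferenced.
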NiagaraSystem /BRCosmetics/Animation/Game/MainPlayer/Emotes/OvenDrastic/FX/NS_OvenDrastic_impact1
dependencies {
 usedDataInterfaces [&NiagaraDataInterfaceCurve, &NiagaraDataInterfaceColorCurve, &NiagaraDataInterfaceArrayFloat]
 usedDynamicInputs []
 usedModules [&SystemState]
}
emitters {
 AirGlow {
  emitterUpdateScript [
   &EmitterState(Life Cycle Mode=::NewEnumerator1, Loop Behavior=::NewEnumerator1, Loop Delay=$User.exp_Delay, Loop Duration=3, UseLoopDelay=true)
   &SpawnBurst_Instantaneous(Spawn Count=1, Spawn Time=0.01, Use Spawn Probability=false)
  ]
  particleSpawnScript [&InitializeParticle(
    Color=&LerpLinearColors(StartColor=<rgba 0.06 0.52 1 1>, EndColor=$User.Color, LerpFactor=$User.UseCustomColor)
    Lifetime Max=0.5
    Lifetime Min=0.35
    Lifetime Mode=::NewEnumerator1
    Sprite Rotation Mode=::NewEnumerator2
    Sprite Size Mode=::NewEnumerator3
    Uniform Sprite Size=180
    UsePositionOffset=true
    Write Material Random=true
   )]
  particleUpdateScript [
   &ParticleState()
   &ScaleSpriteSize(Scale Factor=&Vector2DFromFloat(Value=&FloatFromCurve(
      FloatCurve=DataInterface() {propertyValues {Curve {keys [
          {arriveTangent=2.6083, interpMode=RCIM_Cubic, leaveTangent=0.4318, tangentMode=RCTM_User, time=0, value=0.5}
          {arriveTangent=1.3014, interpMode=RCIM_Cubic, leaveTangent=1.3014, tangentMode=RCTM_User, time=0.3, value=0.8}
          {arriveTangent=0.1331, leaveTangent=0.1331, time=1, value=1.2}
         ]}}}
      Scale Curve=1.25
     )))
   &ScaleColor(
    Scale Alpha=0.2
    Scale RGB=&VectorFromFloat(Value=&FloatFromCurve(
      FloatCurve=DataInterface() {propertyValues {Curve {keys [{interpMode=RCIM_Cubic, tangentMode=RCTM_User, time=0, value=1}, {interpMode=RCIM_Cubic, tangentMode=RCTM_User, time=1, value=0}]}}}
      Scale Curve=4
     ))
    ScaleA=true
    ScaleRGB=true
   )
   &CameraOffset(Camera Offset Amount=20)
   &DynamicMaterialParameters(
    Index 0 Param 1=1
    Index 0 Param 2=1
    Index 0 Param 3=1
    Index 0 Param 4=1
    Module.Param0WriteEnabled=true
    Module.Param1WriteEnabled=true
    Module.Param2WriteEnabled=true
    Module.Param3WriteEnabled=true
   )
  ]
  rendererClasses [&NiagaraSpriteRendererProperties]
  renderers [{rendererClass=&NiagaraSpriteRendererProperties}]
 }
 ArcStraights {
  emitterUpdateScript [
   &EmitterState(Life Cycle Mode=::NewEnumerator1, Loop Behavior=::NewEnumerator1, Loop Delay=0.007, Loop Duration=2, UseLoopDelay=true)
   &SpawnBurst_Instantaneous(Spawn Count=12, Use Spawn Probability=false, UseLoopCountLimit=false)
  ]
  particleSpawnScript [
   &InitializeParticle(
    Color=&RandomRangeLinearColor(Minimum=<rgba 0.7 0.7 0.7 0.4>, Maximum=<rgba 1 1 1 0.6>)
    Lifetime=&RandomRangeFloat(Minimum=0.3, Maximum=0.35)
    Mesh Scale Mode=::NewEnumerator3
    Mesh Uniform Scale=0.175
    UsePositionOffset=false
    Write Material Random=true
   )
   &ShapeLocation(MaskSpawnGroup=false, Offset=<0 15 -10>, Offset Mode=::NewEnumerator0, Override Seed=false, Sphere Radius=0, UseOwnerScale=true)
   &AddVelocity(
    Cone Angle=60
    Cone Axis=<0 0 1>
    Module.UseConeFalloff=false
    Module.UseVelDistribution=false
    Override Seed=false
    Velocity Mode=::NewEnumerator2
    Velocity Speed=10
   )
   &CameraOffset(Camera Offset Amount=5)
   &AddVelocity(Velocity=<10 0 0>) {module=AddVelocity001}
  ]
  particleUpdateScript [
   &ParticleState()
   &UpdateMeshOrientation(
    Facing Direction=&TransformVector(
     Vector=&Add_Vector(A=&Multiply_Vector(A=$Particles.ShapeLocation.ShapeNormal, B=<0.5 0.3 0.3>), B=$Engine.Owner.SystemXAxis)
     Source Space=::World
     Destination Space=::Local
    )
    Facing Interpolation=false
    Orientation Method=::NewEnumerator1
    Override Quat=false
    TransformFacing=false
    TransformUp=false
    Up Direction:Vector3f=<0 0 -1>
   )
   &ScaleMeshSize(Scale Factor=&RandomRangeVector(Minimum=<2.5 2.5 4.5>, Maximum=<3 3 5.5>)) {module=ScaleMeshSize001}
   &ScaleMeshSize(Scale Factor=&MakeVector(
     X=1
     Y=1
     Z=&FloatFromCurve(FloatCurve=DataInterface() {propertyValues {Curve {keys [
          {interpMode=RCIM_Cubic, leaveTangent=0.0966, tangentMode=RCTM_Break, time=0, value=1}
          {arriveTangent=0.0951, interpMode=RCIM_Cubic, leaveTangent=-0.1745, tangentMode=RCTM_Break, time=1, value=1.07}
         ]}}})
    ))
   &DynamicMaterialParameters(
    Index 0 Param 4=&MakeCustomFloatFromBool(
     Bool=false
     True Float=3
     False Float=&SelectFloatFromArray(
      Direct Array Index=&ReturnParticleIDAttributes()
      Float Selection Array=DataInterface() {propertyValues {FloatData [1,0,9,3,10,9]}}
     )
    )
    Index 1 Param 1=0.5
    Index 1 Param 2=2
    Index 1 Param 3=&FloatFromCurve(FloatCurve=DataInterface() {propertyValues {Curve {keys [{interpMode=RCIM_Cubic, tangentMode=RCTM_Break, time=0.5, value=-0.5}, <at 1 0.1>]}}})
    Index 2 Param 1=&Multiply_Float(
     A=&FloatFromCurve(
      FloatCurve=DataInterface() {propertyValues {Curve {keys [{interpMode=RCIM_Cubic, leaveTangent=8.1085, tangentMode=RCTM_Break, time=0.8568, value=1}, {tangentMode=RCTM_User, time=1, value=0}]}}}
      Scale Curve=0.7
     )
     B=&Divide_Float(
      A=&MakeFloatFromVector(Vector=$Particles.Scale, Channel=::NewEnumerator0)
      B=&RandomRangeFloat(Minimum=10, Maximum=15)
     )
    )
    Index 2 Param 2=&FloatFromCurve(
     FloatCurve=DataInterface() {propertyValues {Curve {keys [
         {interpMode=RCIM_Cubic, tangentMode=RCTM_Break, time=0.7, value=0}
         {arriveTangent=-0.0859, interpMode=RCIM_Cubic, tangentMode=RCTM_Break, time=1, value=-1}
        ]}}}
     Scale Curve=50
    )
    Index 2 Param 3=&Waveform(
     Global Input Position=&RandomRangeFloat(Minimum=0, Maximum=15)
     Global Phase=$Particles.Age
     "Waveform [1]"=::NewEnumerator12
     "[1] Phase Scale"=2
     "[1] Frequency"=1
     "[1] Amplitude Min/Max"=<0 1>
    )
    Index 2 Param 4=25
    Index 3 Param 1=&FloatFromCurve(
     FloatCurve=DataInterface() {propertyValues {Curve {keys [
         {interpMode=RCIM_Cubic, leaveTangent=-0.0025, tangentMode=RCTM_Break, time=0, value=0}
         {arriveTangent=1.8528, interpMode=RCIM_Cubic, tangentMode=RCTM_Break, time=1, value=1}
        ]}}}
     Scale Curve=&RandomRangeFloat(Minimum=25, Maximum=35)
    )
    Index 3 Param 2=&FloatFromCurve(FloatCurve=DataInterface() {propertyValues {Curve {keys [
         {interpMode=RCIM_Cubic, leaveTangent=-0.0025, tangentMode=RCTM_Break, time=0, value=0.5}
         {arriveTangent=-0.0316, interpMode=RCIM_Cubic, tangentMode=RCTM_Break, time=1, value=-0.2}
        ]}}})
    Index 3 Param 3=&FloatFromCurve(FloatCurve=DataInterface() {propertyValues {Curve {keys [
         {interpMode=RCIM_Cubic, leaveTangent=1.1034, tangentMode=RCTM_Break, time=0, value=-1}
         {arriveTangent=2.2399, tangentMode=RCTM_User, time=0.5, value=0.125}
        ]}}})
    Index 3 Param 4=&FloatFromCurve(FloatCurve=DataInterface() {propertyValues {Curve {keys [
         {interpMode=RCIM_Cubic, leaveTangent=3.0601, tangentMode=RCTM_Break, time=0.5, value=-0.5}
         {arriveTangent=3.2053, tangentMode=RCTM_User, time=1, value=1}
        ]}}})
    Module.Index0Param0Write=true
    Module.Index0Param1Write=true
    Module.Index0Param2Write=true
    Module.Index0Param3Write=true
    Module.Index1Param0Write=true
    Module.Index1Param1Write=true
    Module.Index1Param2Write=true
    Module.Index1Param3Write=true
    Module.Index2Param0Write=true
    Module.Index2Param1Write=true
    Module.Index2Param2Write=true
    Module.Index2Param3Write=true
    Module.Index3Param0Write=true
    Module.Index3Param1Write=true
    Module.Index3Param2Write=true
    Module.Index3Param3Write=true
    Write Parameter Index 1=true
    Write Parameter Index 2=true
    Write Parameter Index 3=true
   )
   &SolveForcesAndVelocity(Clamp Velocity=false, Limit Acceleration=false)
   &ScaleColor(
    Scale Alpha=&FloatFromCurve(FloatCurve=DataInterface() {propertyValues {Curve {keys [
         {tangentMode=RCTM_User, time=0, value=1}
         {arriveTangent=-0.0405, interpMode=RCIM_Cubic, leaveTangent=-0.0405, tangentMode=RCTM_User, time=0.7009, value=1}
         {arriveTangent=-8.0089, interpMode=RCIM_Cubic, leaveTangent=-8.0089, tangentMode=RCTM_User, time=1, value=0}
        ]}}})
    Scale RGB=&Lerp_Vector(
     A=&VectorFromFloat(Value=&FloatFromCurve(FloatCurve=DataInterface() {propertyValues {Curve {keys [<at 0.2771 1>,<at 0.3786 1>]}}}, Scale Curve=20))
     B=&VectorFromFloat(Value=&RandomRangeFloat(Minimum=0.3, Maximum=2, Evaluation Type=::NewEnumerator1))
     Alpha=&FloatFromCurve(FloatCurve=DataInterface() {propertyValues {Curve {keys [
          {interpMode=RCIM_Cubic, leaveTangent=2.5294, tangentMode=RCTM_Break, time=0, value=0}
          {arriveTangent=-0.2166, interpMode=RCIM_Cubic, tangentMode=RCTM_Break, time=0.4, value=1}
          {arriveTangent=-1.7016, interpMode=RCIM_Cubic, leaveTangent=-1.7016, time=0.8143, value=1}
          <at 0.9877 0 RCIM_Cubic>
         ]}}})
    )
    ScaleA=true
    ScaleRGB=true
   )
   /SetVariables(Particles.MeshOrientation=&MultiplyQuaternion(
     Quaternion A=&MultiplyQuaternion(
      Quaternion A=&MakeQuatFromAxisAngle(Axis=<0 0 1>, AngleInDegrees=0)
      Quaternion B=&FindQuatBetweenTwoVectors(From Vector=<0 0 0>, To Vector=$Particles.Velocity)
     )
     Quaternion B=<0 0 0 1>
    )) {enabled=false, isSetParameters=true}
  ]
  rendererClasses [&NiagaraMeshRendererProperties]
  renderers [
   {rendererClass=&NiagaraMeshRendererProperties, rendererIndex=0}
   {rendererClass=&NiagaraMeshRendererProperties, rendererIndex=1}
  ]
 }
 Flare {
  emitterUpdateScript [
   &EmitterState(
    Inactive Response=::NewEnumerator1
    Life Cycle Mode=::NewEnumerator1
    Loop Behavior=::NewEnumerator1
    Loop Delay=$User.exp_Delay
    Loop Duration=3
    UseLoopDelay=true
   )
   &SpawnBurst_Instantaneous(Use Spawn Probability=false, UseLoopCountLimit=true)
  ]
  particleSpawnScript [
   &InitializeParticle(
    Color=<rgba 0.4334 0.4037 1 1>
    Lifetime Max=0.1667
    Lifetime Min=0.1333
    Lifetime Mode=::NewEnumerator1
    Sprite Size Mode=::NewEnumerator1
    Uniform Sprite Size Max=185
    Uniform Sprite Size Min=170
    UsePositionOffset=true
    Write Material Random=true
   )
   &SpriteFacingAndAlignment(Sprite Alignment=<0 0 1>, Sprite Facing=<0 0 1>, WriteAlignment=true, WriteFacing=false)
   &CameraOffset(Camera Offset Amount=60)
   &DynamicMaterialParameters(Write Parameter Index 0=false)
  ]
  particleUpdateScript [
   &ParticleState()
   &ScaleColor(
    Scale Alpha=&FloatFromCurve(FloatCurve=DataInterface() {propertyValues {Curve {keys [
         {arriveTangent=6.6296, interpMode=RCIM_Cubic, leaveTangent=6.6296, tangentMode=RCTM_User, time=0, value=0}
         {interpMode=RCIM_Cubic, tangentMode=RCTM_User, time=0.25, value=1}
         {tangentMode=RCTM_User, time=1, value=0}
        ]}}})
    ScaleA=true
    ScaleRGB=true
   )
   &ScaleSpriteSize(Uniform Curve Sprite Scale=DataInterface() {propertyValues {Curve {keys [
        {arriveTangent=7.5797, interpMode=RCIM_Cubic, leaveTangent=7.5797, tangentMode=RCTM_User, time=0, value=0.25}
        {interpMode=RCIM_Cubic, tangentMode=RCTM_User, time=0.25, value=1}
        {arriveTangent=-1.3776, interpMode=RCIM_Cubic, leaveTangent=-1.3776, tangentMode=RCTM_User, time=1, value=0.5}
       ]}}})
   /SetVariables(
    Particles.Color Wide:LinearColor=&MultiplyLinearColor(Linear Color=$Particles.Color, Scale Factor=<0.5 0.5 0.5 1>)
    Particles.SpriteSize_Wide:Vector2f=&Multiply_Vector2D(A=$Particles.SpriteSize, B=<2 0.15>)
   ) {isSetParameters=true}
   /SetVariables(Particles.Position_Light:NiagaraPosition=&Add_Vector(A=Linked() {linkedVariable {name=Particles.Initial.Position, type=NiagaraPosition}}, B=<0 0 0>)) {isSetParameters=true}
   &FN_Light_Attributes(
    Color=$User.Color_Blade_B
    Create Color Attribute=true
    Create Exponent Attribute=true
    Create Radius Attribute=true
    Exponent=2.5
    Radius=&Multiply_Float(A=&MakeFloatFromVector2d(Vector 2D=$Particles.SpriteSize), B=1.3)
    Scale Color=&FloatFromCurve(
     FloatCurve=DataInterface() {propertyValues {Curve {keys [{interpMode=RCIM_Cubic, tangentMode=RCTM_User, time=0.25, value=1}, {interpMode=RCIM_Cubic, tangentMode=RCTM_User, time=1, value=0}]}}}
     Scale Curve=3.5
    )
   )
  ]
  rendererClasses [&NiagaraSpriteRendererProperties, &NiagaraLightRendererProperties]
  renderers [
   {rendererClass=&NiagaraSpriteRendererProperties}
   {rendererClass=&NiagaraSpriteRendererProperties, rendererIndex=1}
   {rendererClass=&NiagaraLightRendererProperties, rendererIndex=2}
  ]
 }
 Flash001 {
  emitterUpdateScript [
   &EmitterState(Life Cycle Mode=::NewEnumerator1, Loop Behavior=::NewEnumerator1, Loop Delay=$User.exp_Delay, Loop Duration=3, UseLoopDelay=true)
   &SpawnBurst_Instantaneous(Use Spawn Probability=false, UseLoopCountLimit=false)
  ]
  particleSpawnScript [
   &InitializeParticle(
    Color=<rgba 1.3594 2.4773 6 1>
    Lifetime=0.27
    Mesh Scale Mode=::NewEnumerator3
    Mesh Uniform Scale=0.4
    Sprite Size Mode=::NewEnumerator3
    Uniform Sprite Size=120
    UsePositionOffset=true
    Write Material Random=true
   )
   &CameraOffset(Camera Offset Amount=50)
  ]
  particleUpdateScript [
   &ParticleState()
   &ScaleMeshSize(Scale Factor=&VectorFromFloat(Value=&FloatFromCurve(FloatCurve=DataInterface() {propertyValues {Curve {keys [
          {arriveTangent=1.5171, interpMode=RCIM_Cubic, leaveTangent=1.5171, tangentMode=RCTM_User, time=0, value=0.4}
          {arriveTangent=0.0224, interpMode=RCIM_Cubic, leaveTangent=0.0224, tangentMode=RCTM_User, time=1, value=1}
         ]}}})))
   &DynamicMaterialParameters(
    Index 0 Param 1=1
    Index 0 Param 2=1
    Index 1 Param 1=100
    Index 3 Param 1=&FloatFromCurve(FloatCurve=DataInterface() {propertyValues {Curve {keys [
         {interpMode=RCIM_Cubic, tangentMode=RCTM_User, time=0.5, value=0}
         {arriveTangent=2.1268, interpMode=RCIM_Cubic, leaveTangent=2.1268, tangentMode=RCTM_User, time=1, value=1}
        ]}}})
    Index 3 Param 2=&FloatFromCurve(FloatCurve=DataInterface() {propertyValues {Curve {keys [{arriveTangent=-2.5304, interpMode=RCIM_Cubic, time=0, value=2.5}, {interpMode=RCIM_Cubic, leaveTangent=-3.141, time=1, value=0.4}]}}})
    Index 3 Param 3=&FloatFromCurve(FloatCurve=DataInterface() {propertyValues {Curve {keys [
         {arriveTangent=-1.2425, interpMode=RCIM_Cubic, time=0, value=1}
         {arriveTangent=-4.4048, interpMode=RCIM_Cubic, leaveTangent=-4.4048, tangentMode=RCTM_User, time=1, value=-0.7}
        ]}}})
    Index 3 Param 4=&FloatFromCurve(
     FloatCurve=DataInterface() {propertyValues {Curve {keys [
         {arriveTangent=2.3563, interpMode=RCIM_Cubic, leaveTangent=2.3563, tangentMode=RCTM_User, time=0, value=0}
         {tangentMode=RCTM_User, time=1, value=1}
        ]}}}
     Scale Curve=3
    )
    Module.Index0Param0Write=true
    Module.Index0Param1Write=true
    Module.Index0Param2Write=true
    Module.Index0Param3Write=true
    Module.Index1Param0Write=true
    Module.Index1Param1Write=false
    Module.Index1Param2Write=false
    Module.Index1Param3Write=false
    Module.Index3Param0Write=true
    Module.Index3Param1Write=true
    Module.Index3Param2Write=true
    Module.Index3Param3Write=true
    Write Parameter Index 1=true
    Write Parameter Index 3=true
   )
   &ScaleColor(
    Scale Alpha=&FloatFromCurve(
     FloatCurve=DataInterface() {propertyValues {Curve {keys [
         {arriveTangent=-2.4409, arriveTangentWeight=0.162, interpMode=RCIM_Cubic, leaveTangentWeight=0.2268, time=0.2, value=1}
         {arriveTangent=-3.5548, interpMode=RCIM_Cubic, leaveTangent=-3.5548, tangentMode=RCTM_User, time=0.5, value=0.5}
         <at 1 0>
        ]}}}
     Scale Curve=2.5
    )
    Scale RGB=&VectorFromFloat(Value=0.5)
    ScaleA=true
    ScaleRGB=true
   )
   /SetVariables(Particles.LightColor=&MultiplyLinearColor(Linear Color=$Particles.Color, Scale Factor=&MakeVector4_FromFloat(Float=2))) {isSetParameters=true}
   /SetVariables(Particles.RingColor:LinearColor=&MakeLinearColorFromVectorAndFloat(
     "Vector (RGB)"=&MakeVectorFromLinearColorRGB(Color=$Particles.Initial.Color)
     "Float (Alpha)"=&FloatFromCurve(FloatCurve=DataInterface() {propertyValues {Curve {keys [<at 0 1 RCIM_Cubic>,<at 1 0>]}}}, Scale Curve=0.1)
    )) {isSetParameters=true}
   &ScaleSpriteSize(Uniform Curve Sprite Scale=DataInterface() {propertyValues {Curve {keys [
        {arriveTangent=6.7629, interpMode=RCIM_Cubic, leaveTangent=6.7629, tangentMode=RCTM_User, time=0, value=0.75}
        {arriveTangent=-0.7511, interpMode=RCIM_Cubic, leaveTangent=-0.7511, time=0.11, value=1}
        <at 0.9986 0>
       ]}}})
  ]
  rendererClasses [&NiagaraSpriteRendererProperties, &NiagaraMeshRendererProperties]
  renderers [
   {rendererClass=&NiagaraSpriteRendererProperties}
   {rendererClass=&NiagaraMeshRendererProperties, rendererIndex=1}
  ]
 }
 Sparks_ {
  simTarget=GPUComputeSim
  emitterUpdateScript [
   &EmitterState(
    Enable Distance Culling=true
    Life Cycle Mode=::NewEnumerator1
    Loop Behavior=::NewEnumerator1
    Loop Delay=$User.exp_Delay
    Loop Duration=1
    MaxDistance=3000
    Module.bUseMaxDistance=true
    Module.bUseMinDistance=false
    Scalability Mode=::NewEnumerator1
    UseLoopDelay=true
   )
   &SpawnBurst_Instantaneous(Spawn Count=&RandomRangeInt(Minimum=20, Maximum=25), Use Spawn Probability=false)
  ]
  particleSpawnScript [
   &InitializeParticle(
    Color=<rgba 0 0.6677 1 1>
    Lifetime Max=0.3
    Lifetime Min=0.25
    Lifetime Mode=::NewEnumerator1
    Sprite Size Max=<1.5 5>
    Sprite Size Min=<0.5 3>
    Sprite Size Mode=::NewEnumerator2
    UsePositionOffset=true
    Write Material Random=true
   )
   &ShapeLocation(
    MaskSpawnGroup=false
    Override Seed=false
    Sphere Radius=&RandomRangeFloat(Minimum=10, Maximum=30)
    UseOwnerScale=true
   )
   &AddVelocity(
    Velocity=$Particles.ShapeLocation.ShapeVector
    Velocity Speed Scale=&RandomRangeFloat(Minimum=5, Maximum=15)
   ) {module=AddVelocity001}
   &AddVelocity(Velocity=&RandomRangeVector(Minimum=<450 0 300>, Maximum=<-450 0 400>), Velocity Speed Scale=0.75)
  ]
  particleUpdateScript [
   &ParticleState()
   &ScaleSpriteSize(
    Scale Factor=&MakeVector2D(
     X=1
     Y=&FloatFromCurve(FloatCurve=DataInterface() {propertyValues {Curve {keys [<at 0.5 1>,<at 1 0>]}}})
    )
    Scale Sprite Size Mode=::NewEnumerator2
   )
   &Drag(
    Drag=&FloatFromCurve(FloatCurve=DataInterface() {propertyValues {Curve {defaultValue=1, keys [{tangentMode=RCTM_User, time=0, value=0}, {tangentMode=RCTM_User, time=1, value=3}]}}})
    Use Linear Drag=true
    Use Rotational Drag=false
   )
   &GravityForce()
   &SolveForcesAndVelocity(Clamp Velocity=false, Limit Acceleration=false)
   &ScaleSpriteSizeBySpeed(Max Scale Factor=<0.75 5>)
   &ScaleColor(
    Scale Alpha=&FloatFromCurve(FloatCurve=DataInterface() {propertyValues {Curve {keys [{interpMode=RCIM_Cubic, tangentMode=RCTM_User, time=0.5, value=1}, {interpMode=RCIM_Cubic, tangentMode=RCTM_User, time=1, value=0}]}}})
    Scale RGB=&Lerp_Vector(
     A=<20 20 20>
     B=<2 2 2>
     Alpha=&FloatFromCurve(FloatCurve=DataInterface() {propertyValues {Curve {keys [{interpMode=RCIM_Cubic, tangentMode=RCTM_User, time=0, value=1}, {interpMode=RCIM_Cubic, tangentMode=RCTM_User, time=0.5, value=0}]}}})
    )
    ScaleA=true
    ScaleRGB=true
   )
  ]
  rendererClasses [&NiagaraSpriteRendererProperties]
  renderers [{rendererClass=&NiagaraSpriteRendererProperties}]
 }
}
summary {
 systemName=NS_OvenDrastic_impact1
 emitters [
  {emitterName=AirGlow, rendererClasses [&NiagaraSpriteRendererProperties]}
  {emitterName=Sparks_, simTarget=GPUComputeSim, rendererClasses [&NiagaraSpriteRendererProperties]}
  {emitterName=Flash001, rendererClasses [&NiagaraSpriteRendererProperties, &NiagaraMeshRendererProperties]}
  {emitterName=Flare, rendererClasses [&NiagaraSpriteRendererProperties, &NiagaraLightRendererProperties]}
  {emitterName=ArcStraights, rendererClasses [&NiagaraMeshRendererProperties]}
 ]
 userVariables [
  User.Color {type=LinearColor, description="", defaultValue {struct=LinearColor, value=<rgba 1.3594 2.4773 6 1>}}
  User.Color_Blade_B {type=LinearColor, description="", defaultValue {struct=LinearColor, value=<rgba 0.2266 0.4129 1 1>}}
  User.exp_Delay {type=NiagaraFloat, defaultValue=0.0325, description=""}
  User.UseCustomColor {type=NiagaraFloat, defaultValue=0, description=""}
  User.UseCustomColor1 {type=NiagaraFloat, defaultValue=0, description=""}
 ]
}
